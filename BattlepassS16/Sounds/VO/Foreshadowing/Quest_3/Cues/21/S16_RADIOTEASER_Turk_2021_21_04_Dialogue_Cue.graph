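# Sound cue for the Turkish radio-teaser voice line 21_04 from the Season 16 foreshadowing Quest 3 dialogue set.
# A single dialogue player node plays the matching Dialogue Wave with the S16QuestVO speaker and no targets.
# The cue routes through the Fort_VO_Narrator sound class and uses the shared S17 foreshadowing quest VO attenuation to limit its audible range.
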SoundCue /BattlepassS16/Sounds/VO/Foreshadowing/Quest_3/Cues/21/S16_RADIOTEASER_Turk_2021_21_04_Dialogue_Cue
root=DialoguePlayer_0
nodes [{
  id=DialoguePlayer_0
  properties {dialogueWaveParameter {
    dialogueWave=/BattlepassS16/Sounds/VO/Foreshadowing/Quest_3/Waves/Turk/S16_RADIOTEASER_Turk_2021_21_04_Dialogue
    context {speaker=/BattlepassS16/Sounds/VO/Foreshadowing/S16QuestVO, targets []}
   }}
 }]
properties {
 attenuationSettings=/BattlepassS16/Sounds/VO/Foreshadowing/S17_Foreshadowing_Quest_VO_Att
 maxDistance=5400
 soundClassObject=/Game/Sounds/FortSoundClassesAndModes/Fort_VO_Narrator
 volumeMultiplier=1
}
